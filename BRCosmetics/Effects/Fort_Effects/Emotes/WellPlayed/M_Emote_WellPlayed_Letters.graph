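# Lit material for the letter mesh particles in the Well Played emote effect, used with Niagara.
# An inverted Fresnel with brightness and exponent controls multiplies particle color scaled by an OuterGlow_Intensity dynamic parameter, forming a view-dependent glow.
# A WPO_Offset vector parameter is transformed from camera space into a named WPO reroute, indicating a camera-relative vertex offset.
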Material /BRCosmetics/Effects/Fort_Effects/Emotes/WellPlayed/M_Emote_WellPlayed_Letters
graph {
 properties {bTangentSpaceNormal=false, bUsedWithNiagaraMeshParticles=true, pixelDepthOffsetMode=PDOM_Legacy, usedShadingModels=Default Lit}
 nodes [
  Constant()
  Constant() {r=0.9}
  Constant() {r=0.1}
  Multiply(
   A=MaterialFunctionCall(Brightness=ScalarParameter() {defaultValue=1, parameterName=Brightness}, Exponent=ScalarParameter() {defaultValue=1, parameterName=Exp}) {materialFunction=&MF_Fresnel_FX}.Fresnel Inverted
   B=Multiply(
    A=ParticleColor().RGB
    B=DynamicParameter() {defaultValue=<rgba 1 0.4 0 0>, paramNames [OuterGlow_Intensity, Param2, Param3, Param4]}.OuterGlow_Intensity
   )
  )
  NamedRerouteDeclaration(None=Transform(None=VectorParameter() {parameterName=WPO_Offset}.RGB) {transformSourceType=TRANSFORMSOURCE_Camera}) {name=WPO, nodeColor=<rgba 1 0.0938 0 1>}
  NamedRerouteUsage()
  ParticleColor()
 ]
}
