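# Additive unlit flare material for Niagara sprite and mesh particles in the WellPlayed emote effect.
# A radial sphere mask, sharpened by a dynamic-parameter power and tinted by particle color, feeds the shared emissive adjustment function.
# Particle alpha, depth-faded except on mobile, is multiplied by a flicker from four multiplied sine waves, with per-particle speed and blend.
# A vector parameter transformed from camera space provides a world position offset.
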Material /BRCosmetics/Effects/Fort_Effects/Emotes/WellPlayed/M_WellPlayed_Flare
graph {
 properties {
  bAllowNegativeEmissiveColor=true
  bUsedWithNiagaraMeshParticles=true
  bUsedWithNiagaraSprites=true
  blendMode=BLEND_Additive
  pixelDepthOffsetMode=PDOM_Legacy
  shadingModel=MSM_Unlit
  usedShadingModels=Unlit
 }
 nodes [
  n1 = SphereMask(
   A=TextureCoordinate()
   B=Constant() {r=0.5}
   Radius=ScalarParameter() {defaultValue=0.5, parameterName=Radius}
   Hardness=ScalarParameter() {parameterName=Hardness}
  ) {hardnessPercent=0}
  n2 = Clamp(None=DynamicParameter() {paramNames [FlickerFade, FlickerSpeed, Sphere Mask Power, Param4]}.Sphere Mask Power) {clampMode=CMODE_ClampMin, maxDefault=100, minDefault=1}
  n3 = ParticleColor()
  MaterialFunctionCall(Emissive=Multiply(A=Power(Base=@n1, Exp=@n2), B=@n3.RGB)) {materialFunction=&MF_FX_EmissiveAdjustments}
  n4 = ScalarParameter() {defaultValue=30, parameterName=DepthFade}
  n5 = DynamicParameter() {paramNames [FlickerFade, FlickerSpeed, Sphere Mask Power, Param4]}
  n6 = Multiply(A=Time(), B=@n5.FlickerSpeed)
  n7 = Sine(None=@n6) {period=0.155}
  n8 = Sine(None=@n6) {period=0.2}
  n9 = Multiply(A=Sine(None=@n6) {period=0.075}, B=Multiply(A=Multiply(A=@n7, B=@n8), B=Sine(None=@n6) {period=0.43}))
  Multiply(
   A=ShadingPathSwitch(Default=DepthFade(Opacity=@n3.A, FadeDistance=@n4), Mobile=@n3.A)
   B=LinearInterpolate(A=Clamp(None=@n9) {maxDefault=3}, Alpha=Saturate(None=@n5.FlickerFade)) {constA=1}
  )
  NamedRerouteDeclaration(None=Transform(None=VectorParameter() {parameterName=WPO_Offset}.RGB) {transformSourceType=TRANSFORMSOURCE_Camera}) {name=WPO, nodeColor=<rgba 1 0.0938 0 1>}
  NamedRerouteUsage()
 ]
}
