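# Niagara effect for the Well Played emote that pops "Good", "Game", "Well" and "Played" in sequence as meshes with sprites.
# User.NormalizedNotifyProgress drives their scale, brightness and closing spin, fall and shrink.
# System-level random indices pick each word's color from a shared palette, with intensity adjusted for time of day.
# Hand_Glows pins sprites to both hand bones, blending through the word colors using nested lerps.
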NiagaraSystem /BRCosmetics/Effects/Fort_Effects/Emotes/WellPlayed/NS_Emote_WellPlayed
dependencies {
 usedDataInterfaces [
  &NiagaraDataInterfaceCurve
  &NiagaraDataInterfaceArrayFloat
  &NiagaraDataInterfaceSkeletalMesh
  &NiagaraDataInterfaceArrayColor
 ]
 usedDynamicInputs [
  &RandomRangeInt
  &MakeCustomIntFromBool
  &SetBoolByIntComparison
  &Add_Integer
 ]
 usedModules [
  &self:SystemSpawnScript.SystemScriptSource.SystemScriptGraph.NiagaraNodeAssignment_7.SetVariables_#1
  &self:SystemSpawnScript.SystemScriptSource.SystemScriptGraph.NiagaraNodeAssignment_8.SetVariables_#2
  &self:SystemSpawnScript.SystemScriptSource.SystemScriptGraph.NiagaraNodeAssignment_9.SetVariables_#3
  &self:SystemSpawnScript.SystemScriptSource.SystemScriptGraph.NiagaraNodeAssignment_10.SetVariables_#4
  &self:SystemSpawnScript.SystemScriptSource.SystemScriptGraph.NiagaraNodeAssignment_11.SetVariables_#5
  &self:SystemSpawnScript.SystemScriptSource.SystemScriptGraph.NiagaraNodeAssignment_14.SetVariables_#6
  &SystemState
 ]
}
emitters {
 Game {
  emitterUpdateScript [
   &EmitterState()
   &SpawnBurst_Instantaneous(Use Spawn Probability=false, UseLoopCountLimit=false)
   /SetVariables(Emitter.WPO_Offset:Vector3f=<-15 0 0>) {isSetParameters=true}
  ]
  particleSpawnScript [&InitializeParticle(
    Color=&SelectLinearColorFromArray(
     Array Sampling Mode=::NewEnumerator1
     Direct Array Index=$System.Game_Random
     Color Selection Array=$System.MasterColorSelection
     Random Seed=?
    )
    Lifetime=3.5
    Mesh Scale Mode=::NewEnumerator3
    Position Mode=::NewEnumerator0
    Sprite Size Mode=::NewEnumerator3
    Uniform Sprite Size=75
    Write Material Random=true
   )]
  particleUpdateScript [
   &ParticleState()
   &UpdateMeshOrientation(
    Rotation Rate=&FloatFromCurve(
     FloatCurve=DataInterface() {propertyValues {Curve {keys [{arriveTangent=15.5207, interpMode=RCIM_Cubic, leaveTangent=15.5207, tangentMode=RCTM_User, time=0.8, value=0}, <at 1 1 RCIM_Cubic>]}}}
     CurveIndex=$User.NormalizedNotifyProgress
    )
    Rotation Vector=&RandomRangeVector(Minimum=<-1 -1 -1>, Maximum=<1 1 1>)
   )
   &ScaleSpriteSize(
    Uniform Curve Index=$User.NormalizedNotifyProgress
    Uniform Curve Sprite Scale=DataInterface() {propertyValues {Curve {keys [
        {interpMode=RCIM_Cubic, leaveTangentWeight=0.177, tangentMode=RCTM_User, time=0.217, value=0}
        {
         arriveTangent=-39.4794
         arriveTangentWeight=0.0737
         interpMode=RCIM_Cubic
         leaveTangent=-28.7535
         leaveTangentWeight=0.3121
         tangentMode=RCTM_Break
         time=0.3284
         value=1
        }
        {arriveTangent=5.8598, interpMode=RCIM_Cubic, leaveTangent=5.8598, tangentMode=RCTM_User, time=0.3708, value=0.9988}
        {arriveTangent=-15.0705, interpMode=RCIM_Cubic, tangentMode=RCTM_Break, time=0.4, value=1}
        {arriveTangentWeight=0.1572, interpMode=RCIM_Cubic, leaveTangentWeight=0.9381, tangentMode=RCTM_Break, time=0.8, value=1}
        <at 1 0 RCIM_Cubic>
       ]}}}
   )
   &DynamicMaterialParameters(
    Index 0 Param 1=10
    Index 0 Param 2=1
    Module.Param0WriteEnabled=true
    Module.Param1WriteEnabled=true
    Module.Param2WriteEnabled=false
    Module.Param3WriteEnabled=false
   )
   &ScaleMeshSize(Scale Factor=&VectorFromFloat(Value=&FloatFromCurve(
      FloatCurve=DataInterface() {propertyValues {Curve {keys [
          {interpMode=RCIM_Cubic, leaveTangentWeight=0.177, tangentMode=RCTM_User, time=0.217, value=0}
          {
           arriveTangent=-39.4794
           arriveTangentWeight=0.0737
           interpMode=RCIM_Cubic
           leaveTangent=-28.7535
           leaveTangentWeight=0.3121
           tangentMode=RCTM_Break
           time=0.3284
           value=1
          }
          {arriveTangent=5.8598, interpMode=RCIM_Cubic, leaveTangent=5.8598, tangentMode=RCTM_User, time=0.3708, value=0.9988}
          {arriveTangent=-15.0705, interpMode=RCIM_Cubic, tangentMode=RCTM_Break, time=0.4, value=1}
          {arriveTangentWeight=0.1572, interpMode=RCIM_Cubic, leaveTangentWeight=0.9381, tangentMode=RCTM_Break, time=0.8, value=1}
          <at 1 0 RCIM_Cubic>
         ]}}}
      CurveIndex=$User.NormalizedNotifyProgress
     )))
   &ScaleColor(
    Scale RGB=&VectorFromFloat(Value=&Multiply_Float(
      A=&Lerp_Float(
       A=&FNTimeOfDayScalar(Night Scale=6, Morning Scale=8, Day Scale=8, Evening Scale=8)
       B=&FNTimeOfDayScalar(Night Scale=0.5, Morning Scale=1, Evening Scale=1)
       Alpha=&FloatFromCurve(
        FloatCurve=DataInterface() {propertyValues {Curve {keys [
            {arriveTangentWeight=4.8072, interpMode=RCIM_Cubic, leaveTangentWeight=3.4079, tangentMode=RCTM_User, time=0.097, value=0}
            {arriveTangent=27.0151, arriveTangentWeight=0.0162, interpMode=RCIM_Cubic, leaveTangent=-0.7078, tangentMode=RCTM_Break, time=0.217, value=1}
           ]}}}
        CurveIndex=$User.NormalizedNotifyProgress
       )
      )
      B=&FloatFromCurve(
       FloatCurve=DataInterface() {propertyValues {Curve {keys [{arriveTangent=-6.4773, interpMode=RCIM_Cubic, leaveTangent=-6.4773, tangentMode=RCTM_User, time=0.8, value=1}, <at 1 0 RCIM_Cubic>]}}}
       CurveIndex=$User.NormalizedNotifyProgress
      )
     ))
    ScaleA=true
    ScaleRGB=true
   )
   &GravityForce(Gravity=&Multiply_VectorByFloat(
     Vector=<0 0 -980>
     Float=&FloatFromCurve(
      FloatCurve=DataInterface() {propertyValues {Curve {keys [{arriveTangent=6.756, interpMode=RCIM_Cubic, leaveTangent=6.756, tangentMode=RCTM_User, time=0.8, value=0}, <at 1 1 RCIM_Cubic>]}}}
      CurveIndex=$User.NormalizedNotifyProgress
     )
    ))
   &SolveForcesAndVelocity(Clamp Velocity=false, Limit Acceleration=false)
   &CameraOffset(Camera Offset Amount=&FloatFromCurve(
     FloatCurve=DataInterface() {propertyValues {Curve {keys [{arriveTangent=8.7069, interpMode=RCIM_Cubic, leaveTangent=8.7069, tangentMode=RCTM_User, time=0.8, value=0}, <at 1 1 RCIM_Cubic>]}}}
     CurveIndex=$User.NormalizedNotifyProgress
     Scale Curve=-50
    ))
   /SetVariables(
    Particles.Sprite Dynamic Parameters:Vector4f=<1 0 3 0>
    Particles.Sprite Position:NiagaraPosition=&AddVectorToPosition(Position=$Particles.Position, Vector=<0 0 -15>)
    Particles.SpriteColor=&MultiplyLinearColor(
     Linear Color=$Particles.Initial.Color
     Scale Factor=&MakeFloat4FromVectorAndFloat(
      "Vector (RGB)"=&VectorFromFloat(Value=&Multiply_Float(
        A=&SelectFloatFromArray(Direct Array Index=$System.Game_Random, Float Selection Array=$System.Sprite Intensity)
        B=&FNTimeOfDayScalar(Night Scale=0.0625, Morning Scale=0.25, Day Scale=0.25, Evening Scale=0.25)
       ))
      "Float (Alpha)"=&FloatFromCurve(
       FloatCurve=DataInterface() {propertyValues {Curve {keys [
           {arriveTangent=-27.1213, interpMode=RCIM_Cubic, leaveTangent=-27.1213, leaveTangentWeight=0.4553, tangentMode=RCTM_User, time=0.8, value=1}
           {arriveTangentWeight=0.0667, interpMode=RCIM_Cubic, tangentMode=RCTM_User, time=0.85, value=0}
          ]}}}
       CurveIndex=$User.NormalizedNotifyProgress
      )
     )
    )
   ) {isSetParameters=true}
  ]
  rendererClasses [&NiagaraMeshRendererProperties, &NiagaraSpriteRendererProperties]
  renderers [
   {rendererClass=&NiagaraMeshRendererProperties, rendererIndex=0}
   {rendererClass=&NiagaraSpriteRendererProperties, rendererIndex=1}
  ]
 }
 Good {
  emitterUpdateScript [
   &EmitterState()
   &SpawnBurst_Instantaneous(Use Spawn Probability=false, UseLoopCountLimit=false)
   /SetVariables(Emitter.WPO_Offset:Vector3f=<-40 0 0>) {isSetParameters=true}
  ]
  particleSpawnScript [&InitializeParticle(
    Color=&SelectLinearColorFromArray(
     Array Sampling Mode=::NewEnumerator1
     Direct Array Index=$System.Good_Random
     Color Selection Array=$System.MasterColorSelection
     Random Seed=?
    )
    Lifetime=3.5
    Mesh Scale Mode=::NewEnumerator3
    Position Mode=::NewEnumerator0
    Sprite Size Mode=::NewEnumerator3
    Uniform Sprite Size=75
    Write Material Random=true
   )]
  particleUpdateScript [
   &ParticleState()
   &UpdateMeshOrientation(
    Rotation Rate=&FloatFromCurve(
     FloatCurve=DataInterface() {propertyValues {Curve {keys [{arriveTangent=15.5207, interpMode=RCIM_Cubic, leaveTangent=15.5207, tangentMode=RCTM_User, time=0.8, value=0}, <at 1 1 RCIM_Cubic>]}}}
     CurveIndex=$User.NormalizedNotifyProgress
    )
    Rotation Vector=&RandomRangeVector(Minimum=<-1 -1 -1>, Maximum=<1 1 1>)
   )
   &ScaleSpriteSize(
    Uniform Curve Index=$User.NormalizedNotifyProgress
    Uniform Curve Sprite Scale=DataInterface() {propertyValues {Curve {keys [
        {interpMode=RCIM_Cubic, leaveTangentWeight=13.4929, tangentMode=RCTM_User, time=0.097, value=0}
        {
         arriveTangent=-40.9128
         arriveTangentWeight=1.5756
         interpMode=RCIM_Cubic
         leaveTangent=-26.1019
         leaveTangentWeight=0.2375
         tangentMode=RCTM_Break
         time=0.217
         value=1
        }
        {
         arriveTangent=14.775
         arriveTangentWeight=0.13
         interpMode=RCIM_Cubic
         leaveTangent=14.775
         leaveTangentWeight=0.1415
         tangentMode=RCTM_User
         time=0.2542
         value=1
        }
        {
         arriveTangent=-6.1279
         arriveTangentWeight=0.0277
         interpMode=RCIM_Cubic
         leaveTangentWeight=0.1667
         tangentMode=RCTM_Break
         time=0.3
         value=1
        }
        {arriveTangentWeight=0.1943, interpMode=RCIM_Cubic, leaveTangentWeight=0.6547, tangentMode=RCTM_Break, time=0.8, value=1}
        <at 1 0 RCIM_Cubic>
       ]}}}
   )
   &DynamicMaterialParameters(
    Index 0 Param 1=10
    Index 0 Param 2=1
    Module.Param0WriteEnabled=true
    Module.Param1WriteEnabled=true
    Module.Param2WriteEnabled=false
    Module.Param3WriteEnabled=false
   )
   &ScaleMeshSize(Scale Factor=&VectorFromFloat(Value=&FloatFromCurve(
      FloatCurve=DataInterface() {propertyValues {Curve {keys [
          {interpMode=RCIM_Cubic, leaveTangentWeight=13.4929, tangentMode=RCTM_User, time=0.097, value=0}
          {
           arriveTangent=-40.9128
           arriveTangentWeight=1.5756
           interpMode=RCIM_Cubic
           leaveTangent=-26.1019
           leaveTangentWeight=0.2375
           tangentMode=RCTM_Break
           time=0.217
           value=1
          }
          {
           arriveTangent=14.775
           arriveTangentWeight=0.13
           interpMode=RCIM_Cubic
           leaveTangent=14.775
           leaveTangentWeight=0.1415
           tangentMode=RCTM_User
           time=0.2542
           value=1
          }
          {
           arriveTangent=-6.1279
           arriveTangentWeight=0.0277
           interpMode=RCIM_Cubic
           leaveTangentWeight=0.1667
           tangentMode=RCTM_Break
           time=0.3
           value=1
          }
          {arriveTangentWeight=0.1943, interpMode=RCIM_Cubic, leaveTangentWeight=0.6547, tangentMode=RCTM_Break, time=0.8, value=1}
          <at 1 0 RCIM_Cubic>
         ]}}}
      CurveIndex=$User.NormalizedNotifyProgress
     )))
   &ScaleColor(
    Scale RGB=&VectorFromFloat(Value=&Multiply_Float(
      A=&Lerp_Float(
       A=&FNTimeOfDayScalar(Night Scale=6, Morning Scale=8, Day Scale=8, Evening Scale=8)
       B=&FNTimeOfDayScalar(Night Scale=0.5, Morning Scale=1, Evening Scale=1)
       Alpha=&FloatFromCurve(
        FloatCurve=DataInterface() {propertyValues {Curve {keys [
            {arriveTangentWeight=4.8072, interpMode=RCIM_Cubic, leaveTangentWeight=3.4079, tangentMode=RCTM_User, time=0.097, value=0}
            {arriveTangent=27.0151, arriveTangentWeight=0.0162, interpMode=RCIM_Cubic, leaveTangent=-0.7078, tangentMode=RCTM_Break, time=0.217, value=1}
           ]}}}
        CurveIndex=$User.NormalizedNotifyProgress
       )
      )
      B=&FloatFromCurve(
       FloatCurve=DataInterface() {propertyValues {Curve {keys [{arriveTangent=-6.4773, interpMode=RCIM_Cubic, leaveTangent=-6.4773, tangentMode=RCTM_User, time=0.8, value=1}, <at 1 0 RCIM_Cubic>]}}}
       CurveIndex=$User.NormalizedNotifyProgress
      )
     ))
    ScaleA=true
    ScaleRGB=true
   )
   &GravityForce(Gravity=&Multiply_VectorByFloat(
     Vector=<0 0 -980>
     Float=&FloatFromCurve(
      FloatCurve=DataInterface() {propertyValues {Curve {keys [{arriveTangent=6.756, interpMode=RCIM_Cubic, leaveTangent=6.756, tangentMode=RCTM_User, time=0.8, value=0}, <at 1 1 RCIM_Cubic>]}}}
      CurveIndex=$User.NormalizedNotifyProgress
     )
    ))
   &SolveForcesAndVelocity(Clamp Velocity=false, Limit Acceleration=false)
   &CameraOffset(Camera Offset Amount=&FloatFromCurve(
     FloatCurve=DataInterface() {propertyValues {Curve {keys [{arriveTangent=8.7069, interpMode=RCIM_Cubic, leaveTangent=8.7069, tangentMode=RCTM_User, time=0.8, value=0}, <at 1 1 RCIM_Cubic>]}}}
     CurveIndex=$User.NormalizedNotifyProgress
     Scale Curve=-50
    ))
   /SetVariables(
    Particles.Sprite Dynamic Parameters:Vector4f=<1 0 3 0>
    Particles.Sprite Position:NiagaraPosition=&AddVectorToPosition(Position=$Particles.Position, Vector=<0 0 -15>)
    Particles.SpriteColor=&MultiplyLinearColor(
     Linear Color=$Particles.Initial.Color
     Scale Factor=&MakeFloat4FromVectorAndFloat(
      "Vector (RGB)"=&VectorFromFloat(Value=&Multiply_Float(
        A=&SelectFloatFromArray(Direct Array Index=$System.Good_Random, Float Selection Array=$System.Sprite Intensity)
        B=&FNTimeOfDayScalar(Night Scale=0.0625, Morning Scale=0.25, Day Scale=0.25, Evening Scale=0.25)
       ))
      "Float (Alpha)"=&FloatFromCurve(
       FloatCurve=DataInterface() {propertyValues {Curve {keys [
           {arriveTangent=-27.1213, interpMode=RCIM_Cubic, leaveTangent=-27.1213, leaveTangentWeight=0.4553, tangentMode=RCTM_User, time=0.8, value=1}
           {arriveTangentWeight=0.0667, interpMode=RCIM_Cubic, tangentMode=RCTM_User, time=0.85, value=0}
          ]}}}
       CurveIndex=$User.NormalizedNotifyProgress
      )
     )
    )
   ) {isSetParameters=true}
  ]
  rendererClasses [&NiagaraMeshRendererProperties, &NiagaraSpriteRendererProperties]
  renderers [
   {rendererClass=&NiagaraMeshRendererProperties, rendererIndex=0}
   {rendererClass=&NiagaraSpriteRendererProperties, rendererIndex=1}
  ]
 }
 Hand_Glows {
  emitterUpdateScript [
   &EmitterState()
   &SpawnBurst_Instantaneous(Use Spawn Probability=false, UseLoopCountLimit=false)
   &SpawnBurst_Instantaneous(Spawn Group=1, Use Spawn Probability=false, UseLoopCountLimit=false) {module=SpawnBurst_Instantaneous001}
  ]
  particleSpawnScript [
   &InitializeParticle(Lifetime=3.5, Sprite Size Mode=::NewEnumerator3, Uniform Sprite Size=66, UsePositionOffset=false, Write Material Random=true)
   &CameraOffset(Camera Offset Amount=-5)
   /SetVariables(
    Particles.Game_Color:LinearColor=&SelectLinearColorFromArray(
     Array Sampling Mode=::NewEnumerator1
     Direct Array Index=$System.Game_Random
     Color Selection Array=$System.MasterColorSelection
     Random Seed=?
    )
    Particles.Good_Color:LinearColor=&SelectLinearColorFromArray(
     Array Sampling Mode=::NewEnumerator1
     Direct Array Index=$System.Good_Random
     Color Selection Array=$System.MasterColorSelection
     Random Seed=?
    )
    Particles.Played_Color:LinearColor=&SelectLinearColorFromArray(
     Array Sampling Mode=::NewEnumerator1
     Direct Array Index=$System.Played_Random
     Color Selection Array=$System.MasterColorSelection
     Random Seed=?
    )
    Particles.Well_Color:LinearColor=&SelectLinearColorFromArray(
     Array Sampling Mode=::NewEnumerator1
     Direct Array Index=$System.Well_Random
     Color Selection Array=$System.MasterColorSelection
     Random Seed=?
    )
   ) {isSetParameters=true}
   /SetVariables(
    Particles.Game_Alpha:NiagaraFloat=&SelectFloatFromArray(Direct Array Index=$System.Game_Random, Float Selection Array=DataInterface() {propertyValues {FloatData [0.5,0.25,0.25,0.125]}})
    Particles.Good_Alpha:NiagaraFloat=&SelectFloatFromArray(Direct Array Index=$System.Good_Random, Float Selection Array=DataInterface() {propertyValues {FloatData [0.5,0.25,0.25,0.125]}})
    Particles.Played_Alpha:NiagaraFloat=&SelectFloatFromArray(Direct Array Index=$System.Played_Random, Float Selection Array=DataInterface() {propertyValues {FloatData [0.5,0.25,0.25,0.125]}})
    Particles.Well_Alpha:NiagaraFloat=&SelectFloatFromArray(Direct Array Index=$System.Well_Random, Float Selection Array=DataInterface() {propertyValues {FloatData [0.5,0.25,0.25,0.125]}})
   ) {isSetParameters=true}
  ]
  particleUpdateScript [
   &ParticleState()
   &SkeletalMeshLocation(
    Bone / Socket Index=&SpawnGroup()
    Bone Sampling Mode=::NewEnumerator4
    MaskSpawnGroup=false
    Sampled Position Offset=&MaskVectorBySpawnGroup(Masked Vector=<10 0 0>, Passthrough Vector=<-10 0 0>)
    Skeletal Mesh=DataInterface() {propertyValues {
      DefaultMesh=null
      ExcludeBoneName=root
      OptionalComponentClass=null
      PreviewMesh=/Game/Characters/Player/Male/Male_Avg_Base/Fortnite_M_Avg_Player
      SkinningMode=SkinOnTheFly
      SoftSourceActor=null
      SourceMode=Default
      UvSetIndex=0
      WholeMeshLOD=-1
      bExcludeBone=false
      bReadDeformedGeometry=true
      bRequireCurrentFrameData=false
      ComponentTags []
      FilteredBones [hand_l,hand_r]
      FilteredSockets []
      MeshUserParameter {parameter {name=None, typeDefHandle {registeredTypeIndex=1}, varData []}}
      SamplingRegions []
     }}
   )
   &ScaleColor(
    Scale Alpha=&Multiply_Float(
     A=&Lerp_Float(
      A=$Particles.Good_Alpha
      B=&Lerp_Float(
       A=$Particles.Game_Alpha
       B=&Lerp_Float(
        A=$Particles.Well_Alpha
        B=$Particles.Played_Alpha
        Alpha=&FloatFromCurve(FloatCurve=DataInterface() {propertyValues {Curve {keys [<at 0.335 0 RCIM_Cubic>, {arriveTangent=17.7373, interpMode=RCIM_Cubic, leaveTangent=17.7373, tangentMode=RCTM_User, time=0.514, value=1}]}}})
       )
       Alpha=&FloatFromCurve(FloatCurve=DataInterface() {propertyValues {Curve {keys [<at 0.228 0 RCIM_Cubic>, {arriveTangent=31.5704, interpMode=RCIM_Cubic, leaveTangent=31.5704, tangentMode=RCTM_User, time=0.3284, value=1}]}}})
      )
      Alpha=&FloatFromCurve(
       FloatCurve=DataInterface() {propertyValues {Curve {keys [<at 0.1 0 RCIM_Cubic>, {arriveTangent=26.8029, interpMode=RCIM_Cubic, leaveTangent=26.8029, tangentMode=RCTM_User, time=0.217, value=1}]}}}
       CurveIndex=$User.NormalizedNotifyProgress
      )
     )
     B=&FloatFromCurve(
      FloatCurve=DataInterface() {propertyValues {Curve {keys [{arriveTangent=-7.8516, interpMode=RCIM_Cubic, leaveTangent=-7.8516, tangentMode=RCTM_User, time=0.75, value=1}, <at 1 0 RCIM_Cubic>]}}}
      CurveIndex=$User.NormalizedNotifyProgress
     )
    )
    Scale RGB=&MakeVectorFromLinearColorRGB(Color=&LerpLinearColors(
      StartColor=$Particles.Good_Color
      EndColor=&LerpLinearColors(
       StartColor=$Particles.Game_Color
       EndColor=&LerpLinearColors(
        StartColor=$Particles.Well_Color
        EndColor=$Particles.Played_Color
        LerpFactor=&FloatFromCurve(FloatCurve=DataInterface() {propertyValues {Curve {keys [<at 0.335 0 RCIM_Cubic>, {arriveTangent=17.7373, interpMode=RCIM_Cubic, leaveTangent=17.7373, tangentMode=RCTM_User, time=0.514, value=1}]}}})
       )
       LerpFactor=&FloatFromCurve(FloatCurve=DataInterface() {propertyValues {Curve {keys [<at 0.228 0 RCIM_Cubic>, {arriveTangent=31.5704, interpMode=RCIM_Cubic, leaveTangent=31.5704, tangentMode=RCTM_User, time=0.3284, value=1}]}}})
      )
      LerpFactor=&FloatFromCurve(
       FloatCurve=DataInterface() {propertyValues {Curve {keys [<at 0.1 0 RCIM_Cubic>, {arriveTangent=26.8029, interpMode=RCIM_Cubic, leaveTangent=26.8029, tangentMode=RCTM_User, time=0.217, value=1}]}}}
       CurveIndex=$User.NormalizedNotifyProgress
      )
     ))
    ScaleA=true
    ScaleRGB=true
   )
  ]
  rendererClasses [&NiagaraSpriteRendererProperties]
  renderers [{rendererClass=&NiagaraSpriteRendererProperties}]
 }
 Played {
  emitterUpdateScript [
   &EmitterState()
   &SpawnBurst_Instantaneous(Use Spawn Probability=false, UseLoopCountLimit=false)
   /SetVariables(Emitter.WPO_Offset:Vector3f=<45 0 0>) {isSetParameters=true}
  ]
  particleSpawnScript [&InitializeParticle(
    Color=&SelectLinearColorFromArray(
     Array Sampling Mode=::NewEnumerator1
     Direct Array Index=$System.Played_Random
     Color Selection Array=$System.MasterColorSelection
     Random Seed=?
    )
    Lifetime=3.5
    Mesh Scale Mode=::NewEnumerator3
    Position Mode=::NewEnumerator0
    Sprite Size Mode=::NewEnumerator3
    Uniform Sprite Size=75
    Write Material Random=true
   )]
  particleUpdateScript [
   &ParticleState()
   &UpdateMeshOrientation(
    Rotation Rate=&FloatFromCurve(
     FloatCurve=DataInterface() {propertyValues {Curve {keys [{arriveTangent=15.5207, interpMode=RCIM_Cubic, leaveTangent=15.5207, tangentMode=RCTM_User, time=0.8, value=0}, <at 1 1 RCIM_Cubic>]}}}
     CurveIndex=$User.NormalizedNotifyProgress
    )
    Rotation Vector=&RandomRangeVector(Minimum=<-1 -1 -1>, Maximum=<1 1 1>)
   )
   &ScaleSpriteSize(
    Uniform Curve Index=$User.NormalizedNotifyProgress
    Uniform Curve Sprite Scale=DataInterface() {propertyValues {Curve {keys [
        {interpMode=RCIM_Cubic, leaveTangentWeight=0.177, tangentMode=RCTM_User, time=0.514, value=0}
        {
         arriveTangent=-30.8784
         arriveTangentWeight=0.1731
         interpMode=RCIM_Cubic
         leaveTangent=-19.2048
         leaveTangentWeight=0.1736
         tangentMode=RCTM_Break
         time=0.6568
         value=1
        }
        {arriveTangent=19.4686, interpMode=RCIM_Cubic, leaveTangent=19.4686, tangentMode=RCTM_User, time=0.7052, value=1}
        {arriveTangent=-11.004, interpMode=RCIM_Cubic, tangentMode=RCTM_Break, time=0.7568, value=1}
        {arriveTangentWeight=0.0477, interpMode=RCIM_Cubic, leaveTangentWeight=0.8466, tangentMode=RCTM_Break, time=0.8, value=1}
        {arriveTangentWeight=0.0667, interpMode=RCIM_Cubic, tangentMode=RCTM_User, time=1, value=0}
       ]}}}
   )
   &DynamicMaterialParameters(
    Index 0 Param 1=10
    Index 0 Param 2=1
    Module.Param0WriteEnabled=true
    Module.Param1WriteEnabled=true
    Module.Param2WriteEnabled=false
    Module.Param3WriteEnabled=false
   )
   &ScaleMeshSize(Scale Factor=&VectorFromFloat(Value=&FloatFromCurve(
      FloatCurve=DataInterface() {propertyValues {Curve {keys [
          {interpMode=RCIM_Cubic, leaveTangentWeight=0.177, tangentMode=RCTM_User, time=0.514, value=0}
          {
           arriveTangent=-30.8784
           arriveTangentWeight=0.1731
           interpMode=RCIM_Cubic
           leaveTangent=-19.2048
           leaveTangentWeight=0.1736
           tangentMode=RCTM_Break
           time=0.6568
           value=1
          }
          {arriveTangent=19.4686, interpMode=RCIM_Cubic, leaveTangent=19.4686, tangentMode=RCTM_User, time=0.7052, value=1}
          {arriveTangent=-11.004, interpMode=RCIM_Cubic, tangentMode=RCTM_Break, time=0.7568, value=1}
          {arriveTangentWeight=0.0477, interpMode=RCIM_Cubic, leaveTangentWeight=0.8466, tangentMode=RCTM_Break, time=0.8, value=1}
          {arriveTangentWeight=0.0667, interpMode=RCIM_Cubic, tangentMode=RCTM_User, time=1, value=0}
         ]}}}
      CurveIndex=$User.NormalizedNotifyProgress
     )))
   &ScaleColor(
    Scale RGB=&VectorFromFloat(Value=&Multiply_Float(
      A=&Lerp_Float(
       A=&FNTimeOfDayScalar(Night Scale=6, Morning Scale=8, Day Scale=8, Evening Scale=8)
       B=&FNTimeOfDayScalar(Night Scale=0.5, Morning Scale=1, Evening Scale=1)
       Alpha=&FloatFromCurve(
        FloatCurve=DataInterface() {propertyValues {Curve {keys [
            {arriveTangentWeight=4.8072, interpMode=RCIM_Cubic, leaveTangentWeight=3.4079, tangentMode=RCTM_User, time=0.097, value=0}
            {arriveTangent=27.0151, arriveTangentWeight=0.0162, interpMode=RCIM_Cubic, leaveTangent=-0.7078, tangentMode=RCTM_Break, time=0.217, value=1}
           ]}}}
        CurveIndex=$User.NormalizedNotifyProgress
       )
      )
      B=&FloatFromCurve(
       FloatCurve=DataInterface() {propertyValues {Curve {keys [{arriveTangent=-6.4773, interpMode=RCIM_Cubic, leaveTangent=-6.4773, tangentMode=RCTM_User, time=0.8, value=1}, <at 1 0 RCIM_Cubic>]}}}
       CurveIndex=$User.NormalizedNotifyProgress
      )
     ))
    ScaleA=true
    ScaleRGB=true
   )
   &GravityForce(Gravity=&Multiply_VectorByFloat(
     Vector=<0 0 -980>
     Float=&FloatFromCurve(
      FloatCurve=DataInterface() {propertyValues {Curve {keys [{arriveTangent=6.756, interpMode=RCIM_Cubic, leaveTangent=6.756, tangentMode=RCTM_User, time=0.8, value=0}, <at 1 1 RCIM_Cubic>]}}}
      CurveIndex=$User.NormalizedNotifyProgress
     )
    ))
   &SolveForcesAndVelocity(Clamp Velocity=false, Limit Acceleration=false)
   &CameraOffset(Camera Offset Amount=&FloatFromCurve(
     FloatCurve=DataInterface() {propertyValues {Curve {keys [{arriveTangent=8.7069, interpMode=RCIM_Cubic, leaveTangent=8.7069, tangentMode=RCTM_User, time=0.8, value=0}, <at 1 1 RCIM_Cubic>]}}}
     CurveIndex=$User.NormalizedNotifyProgress
     Scale Curve=-50
    ))
   /SetVariables(
    Particles.Sprite Dynamic Parameters:Vector4f=<1 0 3 0>
    Particles.Sprite Position:NiagaraPosition=&AddVectorToPosition(Position=$Particles.Position, Vector=<0 0 -15>)
    Particles.SpriteColor=&MultiplyLinearColor(
     Linear Color=$Particles.Initial.Color
     Scale Factor=&MakeFloat4FromVectorAndFloat(
      "Vector (RGB)"=&VectorFromFloat(Value=&Multiply_Float(
        A=&SelectFloatFromArray(Direct Array Index=$System.Played_Random, Float Selection Array=$System.Sprite Intensity)
        B=&FNTimeOfDayScalar(Night Scale=0.0625, Morning Scale=0.25, Day Scale=0.25, Evening Scale=0.25)
       ))
      "Float (Alpha)"=&FloatFromCurve(
       FloatCurve=DataInterface() {propertyValues {Curve {keys [
           {arriveTangent=-27.1213, interpMode=RCIM_Cubic, leaveTangent=-27.1213, leaveTangentWeight=0.4553, tangentMode=RCTM_User, time=0.8, value=1}
           {arriveTangentWeight=0.0667, interpMode=RCIM_Cubic, tangentMode=RCTM_User, time=0.85, value=0}
          ]}}}
       CurveIndex=$User.NormalizedNotifyProgress
      )
     )
    )
   ) {isSetParameters=true}
  ]
  rendererClasses [&NiagaraMeshRendererProperties, &NiagaraSpriteRendererProperties]
  renderers [
   {rendererClass=&NiagaraMeshRendererProperties, rendererIndex=0}
   {rendererClass=&NiagaraSpriteRendererProperties, rendererIndex=1}
  ]
 }
 Well {
  emitterUpdateScript [
   &EmitterState()
   &SpawnBurst_Instantaneous(Use Spawn Probability=false, UseLoopCountLimit=false)
   /SetVariables(Emitter.WPO_Offset:Vector3f=<14 0 0>) {isSetParameters=true}
  ]
  particleSpawnScript [&InitializeParticle(
    Color=&SelectLinearColorFromArray(
     Array Sampling Mode=::NewEnumerator1
     Direct Array Index=$System.Well_Random
     Color Selection Array=$System.MasterColorSelection
     Random Seed=?
    )
    Lifetime=3.5
    Mesh Scale Mode=::NewEnumerator3
    Position Mode=::NewEnumerator0
    Sprite Size Mode=::NewEnumerator3
    Uniform Sprite Size=75
    Write Material Random=true
   )]
  particleUpdateScript [
   &ParticleState()
   &UpdateMeshOrientation(
    Rotation Rate=&FloatFromCurve(
     FloatCurve=DataInterface() {propertyValues {Curve {keys [{arriveTangent=15.5207, interpMode=RCIM_Cubic, leaveTangent=15.5207, tangentMode=RCTM_User, time=0.8, value=0}, <at 1 1 RCIM_Cubic>]}}}
     CurveIndex=$User.NormalizedNotifyProgress
    )
    Rotation Vector=&RandomRangeVector(Minimum=<-1 -1 -1>, Maximum=<1 1 1>)
   )
   &ScaleSpriteSize(
    Uniform Curve Index=$User.NormalizedNotifyProgress
    Uniform Curve Sprite Scale=DataInterface() {propertyValues {Curve {keys [
        {interpMode=RCIM_Cubic, leaveTangentWeight=0.177, tangentMode=RCTM_User, time=0.3427, value=0}
        {
         arriveTangent=-30.1722
         arriveTangentWeight=0.1144
         interpMode=RCIM_Cubic
         leaveTangent=-19.393
         leaveTangentWeight=0.2521
         tangentMode=RCTM_Break
         time=0.4854
         value=1
        }
        {arriveTangent=10.8794, interpMode=RCIM_Cubic, leaveTangent=10.8794, tangentMode=RCTM_User, time=0.5335, value=1.0025}
        {arriveTangent=-10.6534, interpMode=RCIM_Cubic, tangentMode=RCTM_Break, time=0.5854, value=1}
        {arriveTangentWeight=0.1049, interpMode=RCIM_Cubic, leaveTangentWeight=0.929, tangentMode=RCTM_Break, time=0.8, value=1}
        {arriveTangentWeight=0.0667, interpMode=RCIM_Cubic, tangentMode=RCTM_User, time=1, value=0}
       ]}}}
   )
   &DynamicMaterialParameters(
    Index 0 Param 1=10
    Index 0 Param 2=1
    Module.Param0WriteEnabled=true
    Module.Param1WriteEnabled=true
    Module.Param2WriteEnabled=false
    Module.Param3WriteEnabled=false
   )
   &ScaleMeshSize(Scale Factor=&VectorFromFloat(Value=&FloatFromCurve(
      FloatCurve=DataInterface() {propertyValues {Curve {keys [
          {interpMode=RCIM_Cubic, leaveTangentWeight=0.177, tangentMode=RCTM_User, time=0.3427, value=0}
          {
           arriveTangent=-30.1722
           arriveTangentWeight=0.1144
           interpMode=RCIM_Cubic
           leaveTangent=-19.393
           leaveTangentWeight=0.2521
           tangentMode=RCTM_Break
           time=0.4854
           value=1
          }
          {arriveTangent=10.8794, interpMode=RCIM_Cubic, leaveTangent=10.8794, tangentMode=RCTM_User, time=0.5335, value=1.0025}
          {arriveTangent=-10.6534, interpMode=RCIM_Cubic, tangentMode=RCTM_Break, time=0.5854, value=1}
          {arriveTangentWeight=0.1049, interpMode=RCIM_Cubic, leaveTangentWeight=0.929, tangentMode=RCTM_Break, time=0.8, value=1}
          {arriveTangentWeight=0.0667, interpMode=RCIM_Cubic, tangentMode=RCTM_User, time=1, value=0}
         ]}}}
      CurveIndex=$User.NormalizedNotifyProgress
     )))
   &ScaleColor(
    Scale RGB=&VectorFromFloat(Value=&Multiply_Float(
      A=&Lerp_Float(
       A=&FNTimeOfDayScalar(Night Scale=6, Morning Scale=8, Day Scale=8, Evening Scale=8)
       B=&FNTimeOfDayScalar(Night Scale=0.5, Morning Scale=1, Evening Scale=1)
       Alpha=&FloatFromCurve(
        FloatCurve=DataInterface() {propertyValues {Curve {keys [
            {arriveTangentWeight=4.8072, interpMode=RCIM_Cubic, leaveTangentWeight=3.4079, tangentMode=RCTM_User, time=0.097, value=0}
            {arriveTangent=27.0151, arriveTangentWeight=0.0162, interpMode=RCIM_Cubic, leaveTangent=-0.7078, tangentMode=RCTM_Break, time=0.217, value=1}
           ]}}}
        CurveIndex=$User.NormalizedNotifyProgress
       )
      )
      B=&FloatFromCurve(
       FloatCurve=DataInterface() {propertyValues {Curve {keys [{arriveTangent=-6.4773, interpMode=RCIM_Cubic, leaveTangent=-6.4773, tangentMode=RCTM_User, time=0.8, value=1}, <at 1 0 RCIM_Cubic>]}}}
       CurveIndex=$User.NormalizedNotifyProgress
      )
     ))
    ScaleA=true
    ScaleRGB=true
   )
   &GravityForce(Gravity=&Multiply_VectorByFloat(
     Vector=<0 0 -980>
     Float=&FloatFromCurve(
      FloatCurve=DataInterface() {propertyValues {Curve {keys [{arriveTangent=6.756, interpMode=RCIM_Cubic, leaveTangent=6.756, tangentMode=RCTM_User, time=0.8, value=0}, <at 1 1 RCIM_Cubic>]}}}
      CurveIndex=$User.NormalizedNotifyProgress
     )
    ))
   &SolveForcesAndVelocity(Clamp Velocity=false, Limit Acceleration=false)
   &CameraOffset(Camera Offset Amount=&FloatFromCurve(
     FloatCurve=DataInterface() {propertyValues {Curve {keys [{arriveTangent=8.7069, interpMode=RCIM_Cubic, leaveTangent=8.7069, tangentMode=RCTM_User, time=0.8, value=0}, <at 1 1 RCIM_Cubic>]}}}
     CurveIndex=$User.NormalizedNotifyProgress
     Scale Curve=-50
    ))
   /SetVariables(
    Particles.Sprite Dynamic Parameters:Vector4f=<1 0 3 0>
    Particles.Sprite Position:NiagaraPosition=&AddVectorToPosition(Position=$Particles.Position, Vector=<0 0 -15>)
    Particles.SpriteColor=&MultiplyLinearColor(
     Linear Color=$Particles.Initial.Color
     Scale Factor=&MakeFloat4FromVectorAndFloat(
      "Vector (RGB)"=&VectorFromFloat(Value=&Multiply_Float(
        A=&SelectFloatFromArray(Direct Array Index=$System.Well_Random, Float Selection Array=$System.Sprite Intensity)
        B=&FNTimeOfDayScalar(Night Scale=0.0625, Morning Scale=0.25, Day Scale=0.25, Evening Scale=0.25)
       ))
      "Float (Alpha)"=&FloatFromCurve(
       FloatCurve=DataInterface() {propertyValues {Curve {keys [
           {arriveTangent=-27.1213, interpMode=RCIM_Cubic, leaveTangent=-27.1213, leaveTangentWeight=0.4553, tangentMode=RCTM_User, time=0.8, value=1}
           {arriveTangentWeight=0.0667, interpMode=RCIM_Cubic, tangentMode=RCTM_User, time=0.85, value=0}
          ]}}}
       CurveIndex=$User.NormalizedNotifyProgress
      )
     )
    )
   ) {isSetParameters=true}
  ]
  rendererClasses [&NiagaraMeshRendererProperties, &NiagaraSpriteRendererProperties]
  renderers [
   {rendererClass=&NiagaraMeshRendererProperties, rendererIndex=0}
   {rendererClass=&NiagaraSpriteRendererProperties, rendererIndex=1}
  ]
 }
}
summary {
 systemName=NS_Emote_WellPlayed
 emitters [
  {emitterName=Good, rendererClasses [&NiagaraMeshRendererProperties, &NiagaraSpriteRendererProperties]}
  {emitterName=Game, rendererClasses [&NiagaraMeshRendererProperties, &NiagaraSpriteRendererProperties]}
  {emitterName=Well, rendererClasses [&NiagaraMeshRendererProperties, &NiagaraSpriteRendererProperties]}
  {emitterName=Played, rendererClasses [&NiagaraMeshRendererProperties, &NiagaraSpriteRendererProperties]}
  {emitterName=Hand_Glows, rendererClasses [&NiagaraSpriteRendererProperties]}
 ]
 userVariables [User.NormalizedNotifyProgress {type=NiagaraFloat, defaultValue=0, description=""}]
}
